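# Behavior tree for the ThawEpson NPC, a leashed creature that chases and melees its goal actor, flees when frightened, and otherwise wanders.
# A priority selector runs frightened flee or wander, return to leash when self and goal are outside it, chase and attack, then idle wander, mostly through tag-driven gameplay abilities.
# Root services update path and leash keys.
# Chasing applies a hostile tag and requires a reachable in-leash goal that is not swimming, muddied or cloaked.
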BehaviorTree /ThawEpsonAI/NPCs/BehaviorTrees/BT_ThawEpson
blackboard=&/FortAIBB_ThawEpson
root {
 name=Selector
 class=BTComposite_Selector
 children [
  Selector {
   class=BTComposite_Selector
   children [
    GameplayAbility {
     class=FortBTTask_ExecuteGameplayAbility
     decorators [
      Has Goal Actor {
       class=BTDecorator_Blackboard
       properties {
        arithmeticOperation=Equal
        intValue=0
        nodeName=Has Goal Actor
        notifyObserver=ValueChange
        stringValue=""
        blackboardKey {selectedKeyName=GoalActor}
       }
      }
      Has Gameplay Ability {
       class=FortBTDecorator_GameplayAbility_HasGameplayAbility
       properties {gameplayAbilityTagContainer {defaultValue {gameplayTags [{tagName=Ability.CosmicThunder.ThawEpson.RunFromGoal}]}}}
      }
      Can Activate Gameplay Ability {
       class=FortBTDecorator_GameplayAbility_CanActivate
       properties {
        flowAbortMode=LowerPriority
        gameplayAbilityTagContainer {defaultValue {gameplayTags [{tagName=Ability.CosmicThunder.ThawEpson.RunFromGoal}]}}
        targetValue {key=GoalActor}
       }
      }
     ]
     properties {nodeName=GameplayAbility, gameplayAbilityTagContainer {defaultValue {gameplayTags [{tagName=Ability.CosmicThunder.ThawEpson.RunFromGoal}]}}}
    }
    GameplayAbility {
     class=FortBTTask_ExecuteGameplayAbility
     decorators [
      Does not have Goal Actor {
       class=BTDecorator_Blackboard
       properties {
        arithmeticOperation=Equal
        basicOperation=NotSet
        intValue=0
        nodeName=Does not have Goal Actor
        notifyObserver=ValueChange
        stringValue=""
        blackboardKey {selectedKeyName=GoalActor}
       }
      }
      Has Gameplay Ability {
       class=FortBTDecorator_GameplayAbility_HasGameplayAbility
       properties {gameplayAbilityTagContainer {defaultValue {gameplayTags [{tagName=NPC.Ability.NonAttack.Wander.Freely}]}}}
      }
      Can Activate Gameplay Ability {
       class=FortBTDecorator_GameplayAbility_CanActivate
       properties {
        flowAbortMode=LowerPriority
        gameplayAbilityTagContainer {defaultValue {gameplayTags [{tagName=NPC.Ability.NonAttack.Wander.Freely}]}}
        targetValue {key=GoalActor}
       }
      }
     ]
     properties {nodeName=GameplayAbility, gameplayAbilityTagContainer {defaultValue {gameplayTags [{tagName=NPC.Ability.NonAttack.Wander.Freely}]}}}
    }
   ]
   decorators [Gameplay Tag Query {
     class=BTDecorator_GameplayTagQuery
     properties {
      flowAbortMode=Both
      nodeName=Gameplay Tag Query
      gameplayTagQuery {
       autoDescription=" ANY( Ability.CosmicThunder.ThawEpson.IsFrightened )"
       queryTokenStream [0,1,1,1,0]
       tagDictionary [{tagName=Ability.CosmicThunder.ThawEpson.IsFrightened}]
      }
     }
    }]
   properties {
    children [
     {
      childComposite=null
      childTask=&self:FortBTTask_ExecuteGameplayAbility_2
      decorators [
       &self:BTDecorator_Blackboard_6
       &self:FortBTDecorator_GameplayAbility_HasGameplayAbility_4
       &self:FortBTDecorator_GameplayAbility_CanActivate_4
      ]
     }
     {
      childComposite=null
      childTask=&self:FortBTTask_ExecuteGameplayAbility_4
      decorators [
       &self:BTDecorator_Blackboard_7
       &self:FortBTDecorator_GameplayAbility_HasGameplayAbility_3
       &self:FortBTDecorator_GameplayAbility_CanActivate_3
      ]
     }
    ]
    services []
   }
   services []
  }
  Selector {
   class=BTComposite_Selector
   children [Go Back To Leash {
     class=FortBTTask_ExecuteGameplayAbility
     decorators []
     properties {nodeName=Go Back To Leash, gameplayAbilityTagContainer {defaultValue {gameplayTags [{tagName=Ability.CosmicThunder.ThawEpson.BackToLeash}]}}}
    }]
   decorators [
    IsNotInLeash {
     class=BTDecorator_Blackboard
     properties {
      arithmeticOperation=Equal
      basicOperation=NotSet
      flowAbortMode=LowerPriority
      intValue=0
      nodeName=IsNotInLeash
      stringValue=""
      blackboardKey {selectedKeyName=IsSelfInLeash}
     }
    }
    IsGoalNotInLeash {
     class=BTDecorator_Blackboard
     properties {
      arithmeticOperation=Equal
      basicOperation=NotSet
      intValue=0
      nodeName=IsGoalNotInLeash
      stringValue=""
      blackboardKey {selectedKeyName=IsGoalInLeash}
     }
    }
    Has Gameplay Ability {
     class=FortBTDecorator_GameplayAbility_HasGameplayAbility
     properties {gameplayAbilityTagContainer {defaultValue {gameplayTags [{tagName=Ability.CosmicThunder.ThawEpson.BackToLeash}]}}}
    }
    Can Activate Gameplay Ability {
     class=FortBTDecorator_GameplayAbility_CanActivate
     properties {
      flowAbortMode=Both
      gameplayAbilityTagContainer {defaultValue {gameplayTags [{tagName=Ability.CosmicThunder.ThawEpson.BackToLeash}]}}
      targetValue {key=GoalActor}
     }
    }
   ]
   properties {
    children [{childComposite=null, childTask=&self:FortBTTask_ExecuteGameplayAbility_1, decorators []}]
    services []
   }
   services []
  }
  Sequence {
   class=BTComposite_Sequence
   children [
    MoveToTarget {
     class=FortBTTask_GameMoveTo
     properties {
      bDeimosFlavor=false
      bFinishMoveOnOverlap=true
      bStopAtGoal=false
      nodeName=MoveToTarget
      acceptableRadius {defaultValue=30}
      bReachTestIncludesAgentRadius {defaultValue=true}
      bReachTestIncludesGoalRadius {defaultValue=true}
      blackboardKey {selectedKeyName=GoalActor}
      filterClass {defaultValue=null}
      navFilterTag {tagName=NavFilter.NPC.DefaultMove}
     }
    }
    Selector {
     class=BTComposite_Selector
     children [
      Attack Target {
       class=FortBTTask_ExecuteGameplayAbility
       decorators [
        Has Gameplay Ability {
         class=FortBTDecorator_GameplayAbility_HasGameplayAbility
         properties {gameplayAbilityTagContainer {defaultValue {gameplayTags [{tagName=Ability.CosmicThunder.ThawEpson.Melee.Attack}]}}}
        }
        Can Activate Gameplay Ability {
         class=FortBTDecorator_GameplayAbility_CanActivate
         properties {
          flowAbortMode=Both
          gameplayAbilityTagContainer {defaultValue {gameplayTags [{tagName=Ability.CosmicThunder.ThawEpson.Melee.Attack}]}}
          targetValue {key=GoalActor}
         }
        }
       ]
       properties {nodeName=Attack Target, gameplayAbilityTagContainer {defaultValue {gameplayTags [{tagName=Ability.CosmicThunder.ThawEpson.Melee.Attack}]}}}
      }
      Wait {class=BTTask_Wait, properties {waitTime {defaultValue=5}}}
     ]
     decorators []
     properties {
      children [
       {
        childComposite=null
        childTask=&self:FortBTTask_ExecuteGameplayAbility_22
        decorators [
         &self:FortBTDecorator_GameplayAbility_HasGameplayAbility_1
         &self:FortBTDecorator_GameplayAbility_CanActivate_1
        ]
       }
       {childComposite=null, childTask=&self:BTTask_Wait_3, decorators []}
      ]
      services []
     }
     services []
    }
   ]
   decorators [
    Has Goal Actor {
     class=BTDecorator_Blackboard
     properties {
      arithmeticOperation=Equal
      intValue=0
      nodeName=Has Goal Actor
      notifyObserver=ValueChange
      stringValue=""
      blackboardKey {selectedKeyName=GoalActor}
     }
    }
    Goal Path exists {
     class=BTDecorator_Blackboard
     properties {
      arithmeticOperation=GreaterOrEqual
      intValue=9
      nodeName=Goal Path exists
      stringValue=Success
      blackboardKey {selectedKeyName=GoalActorPathExists}
     }
    }
    IsGoalInLeash {
     class=BTDecorator_Blackboard
     properties {arithmeticOperation=Equal, intValue=0, nodeName=IsGoalInLeash, stringValue="", blackboardKey {selectedKeyName=IsGoalInLeash}}
    }
    "No Swimming, Muddied or Cloaked Tags" {
     class=BTDecorator_GameplayTagQuery
     properties {
      flowAbortMode=Both
      nodeName="No Swimming, Muddied or Cloaked Tags"
      actorForGameplayTagQuery {selectedKeyName=GoalActor}
      gameplayTagQuery {
       autoDescription=" NONE( MovementMode.SurfaceSwimming, Gameplay.Action.Player.SurfaceSwimming, Granted.Mud.CoveredInMud, Athena.Gameplay.Status.Cloaked )"
       queryTokenStream [0,1,3,4,0,1,2,3]
       tagDictionary [
        {tagName=MovementMode.SurfaceSwimming}
        {tagName=Gameplay.Action.Player.SurfaceSwimming}
        {tagName=Granted.Mud.CoveredInMud}
        {tagName=Athena.Gameplay.Status.Cloaked}
       ]
      }
     }
    }
   ]
   properties {
    children [
     {childTask=&self:FortBTTask_GameMoveTo_15}
     {childComposite=&self:BTComposite_Selector_6, childTask=null}
    ]
    services [&self:FortAthenaBTService_ApplyGameplayTags_0]
   }
   services [Apply Hostile Gameplay Tags {
     class=FortAthenaBTService_ApplyGameplayTags
     properties {
      interval=0.5
      nodeName=Apply Hostile Gameplay Tags
      gameplayTagsToApply {defaultValue {gameplayTags [{tagName=Athena.AI.Status.InsectSwarm.IsHostile}]}}
     }
    }]
  }
  Sequence {
   class=BTComposite_Sequence
   children [
    GameplayAbility {
     class=FortBTTask_ExecuteGameplayAbility
     decorators []
     properties {nodeName=GameplayAbility, gameplayAbilityTagContainer {defaultValue {gameplayTags [{tagName=NPC.Ability.NonAttack.Wander.Freely}]}}}
    }
    Wait {class=BTTask_Wait, properties {waitTime {defaultValue=1}}}
   ]
   decorators [
    Has Gameplay Ability {
     class=FortBTDecorator_GameplayAbility_HasGameplayAbility
     properties {gameplayAbilityTagContainer {defaultValue {gameplayTags [{tagName=NPC.Ability.NonAttack.Wander.Freely}]}}}
    }
    Can Activate Gameplay Ability {
     class=FortBTDecorator_GameplayAbility_CanActivate
     properties {
      flowAbortMode=LowerPriority
      gameplayAbilityTagContainer {defaultValue {gameplayTags [{tagName=NPC.Ability.NonAttack.Wander.Freely}]}}
      targetValue {key=GoalActor}
     }
    }
   ]
   properties {children [
     {childTask=&self:FortBTTask_ExecuteGameplayAbility_0}
     {childTask=&self:BTTask_Wait_4}
    ]}
  }
 ]
 decorators []
 properties {
  children [
   {
    childComposite=&self:BTComposite_Selector_1
    childTask=null
    decorators [&self:BTDecorator_GameplayTagQuery_1]
   }
   {
    childComposite=&self:BTComposite_Selector_0
    childTask=null
    decorators [
     &self:BTDecorator_Blackboard_4
     &self:BTDecorator_Blackboard_3
     &self:FortBTDecorator_GameplayAbility_HasGameplayAbility_2
     &self:FortBTDecorator_GameplayAbility_CanActivate_2
    ]
   }
   {
    childComposite=&self:BTComposite_Sequence_0
    childTask=null
    decorators [
     &self:BTDecorator_Blackboard_0
     &self:BTDecorator_Blackboard_1
     &self:BTDecorator_Blackboard_2
     &self:BTDecorator_GameplayTagQuery_0
    ]
   }
   {
    childComposite=&self:BTComposite_Sequence_5
    childTask=null
    decorators [
     &self:FortBTDecorator_GameplayAbility_HasGameplayAbility_10
     &self:FortBTDecorator_GameplayAbility_CanActivate_9
    ]
   }
  ]
  services [
   &self:FortAthenaBTService_AIEvaluator_0
   &self:BTService_ThawEpsonAI_EvaluateLeash_C_1
  ]
 }
 services [
  AIEvaluator...GoalPathExists {
   class=FortAthenaBTService_AIEvaluator
   properties {
    interval=1
    nodeName=AIEvaluator...GoalPathExists
    aIEvaluatorClass {defaultValue=&/BTEvaluator_ThawEpsonAI_PathExist.BTEvaluator_ThawEpsonAI_PathExist_C}
   }
  }
  Evaluating Leash {
   class=BTService_ThawEpsonAI_EvaluateLeash_C
   properties {
    bCallTickOnSearchStart=false
    bRestartTimerOnEachActivation=false
    bShowEventDetails=false
    bShowPropertyDetails=true
    customDescription=""
    interval=0.5
    nodeName=Evaluating Leash
    randomDeviation=0.1
    actorKeyToTest {bNoneIsAllowedValue=false, selectedKeyName=GoalActor}
    boolToSet {bNoneIsAllowedValue=false, selectedKeyName=IsGoalInLeash}
    selfLeashBoolToSet {bNoneIsAllowedValue=false, selectedKeyName=IsSelfInLeash}
   }
  }
 ]
}
