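# Behavior tree for the ThawEpson NPC AI that runs navigate and idle commands from blackboard execution-status keys, with goal updates suppressed at the root.
# A pending-or-later navigation status runs MoveTo toward the target if it is an actor, otherwise toward the destination position.
# Failing that, a pending-or-later idle status holds a near-endless Wait, with a short Idle wait as fallback.
# A Loop decorator wraps these branches to avoid evaluator restarts.
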BehaviorTree /ThawEpsonAI/NPCs/BehaviorTrees/BT_Vk_ThawEpson_AI
blackboard=&/BB_Vk_ThawEpson
root {
 name=Selector
 class=BTComposite_Selector
 children [Loop to prevent evaluator to restart {
   class=BTComposite_Selector
   children [
    Navigate To {
     class=BTComposite_Selector
     children [
      Navigate to Entity {
       class=FortAthenaBTTask_MoveTo
       children []
       decorators [Is Destination Actor {
         class=BTDecorator_IsBBEntryOfClass
         properties {
          bInverseCondition=false
          flowAbortMode=Both
          nodeName=Is Destination Actor
          blackboardKey {bNoneIsAllowedValue=false, selectedKeyName=Vk_NavigateToTarget}
          testClass {baseClass=&Object, defaultValue=&Actor, key=None}
         }
        }]
       properties {
        bAllowPartialPathToUncompleteNavigationArea=false
        bAllowRestartOnPartialSuccess=false
        bIgnoreRestartSelf=false
        executionStatusKeyName=VK_NavigateToExecutionStatus
        movementResultKeyName=VK_NavigateToResult
        nodeName=Navigate to Entity
        acceptableRadius {defaultValue=5, key=Vk_NavigateToRadius}
        bAllowPartialPath {defaultValue=true, key=Vk_NavigateToAllowPartialPath}
        bAllowStrafe {defaultValue=false, key=VK_NavigateToAllowStrafe}
        bProjectGoalLocation {defaultValue=true, key=None}
        bReachTestIncludesAgentRadius {defaultValue=true, key=None}
        bReachTestIncludesGoalRadius {defaultValue=true, key=None}
        bRequireNavigableEndLocation {defaultValue=true, key=None}
        bStartFromPreviousPath {defaultValue=false, key=None}
        bTrackMovingGoal {defaultValue=true, key=None}
        blackboardKey {bNoneIsAllowedValue=false, selectedKeyName=Vk_NavigateToTarget}
        filterClass {baseClass=&NavigationQueryFilter, defaultValue=null, key=None}
        observedBlackboardValueTolerance {defaultValue=4.75, key=None}
       }
       services []
      }
      Navigate to Position {
       class=FortAthenaBTTask_MoveTo
       children []
       decorators []
       properties {
        bAllowPartialPathToUncompleteNavigationArea=false
        bAllowRestartOnPartialSuccess=false
        bIgnoreRestartSelf=false
        executionStatusKeyName=VK_NavigateToExecutionStatus
        movementResultKeyName=VK_NavigateToResult
        nodeName=Navigate to Position
        acceptableRadius {defaultValue=5, key=Vk_NavigateToRadius}
        bAllowPartialPath {defaultValue=true, key=Vk_NavigateToAllowPartialPath}
        bAllowStrafe {defaultValue=false, key=VK_NavigateToAllowStrafe}
        bProjectGoalLocation {defaultValue=true, key=None}
        bReachTestIncludesAgentRadius {defaultValue=true, key=None}
        bReachTestIncludesGoalRadius {defaultValue=true, key=None}
        bRequireNavigableEndLocation {defaultValue=true, key=None}
        bStartFromPreviousPath {defaultValue=false, key=None}
        bTrackMovingGoal {defaultValue=true, key=None}
        blackboardKey {bNoneIsAllowedValue=false, selectedKeyName=Vk_NavigateToDestination}
        filterClass {baseClass=&NavigationQueryFilter, defaultValue=null, key=None}
        observedBlackboardValueTolerance {defaultValue=4.75, key=None}
       }
       services []
      }
     ]
     decorators [NavigateTo Status {
       class=BTDecorator_Blackboard
       properties {
        arithmeticOperation=GreaterOrEqual
        intValue=3
        nodeName=NavigateTo Status
        stringValue=EExecutionStatus::ExecutionPending
        blackboardKey {selectedKeyName=VK_NavigateToExecutionStatus}
       }
      }]
     properties {
      nodeName=Navigate To
      children [
       {
        childComposite=null
        childTask=&self:FortAthenaBTTask_MoveTo_0
        decorators [&self:BTDecorator_IsBBEntryOfClass_0]
       }
       {childComposite=null, childTask=&self:FortAthenaBTTask_MoveTo_1, decorators []}
      ]
      services []
     }
     services []
    }
    Wait {
     class=BTTask_Wait
     decorators [Blackboard Based Condition {
       class=BTDecorator_Blackboard
       properties {
        arithmeticOperation=GreaterOrEqual
        intValue=3
        nodeName=Blackboard Based Condition
        notifyObserver=ValueChange
        stringValue=ExecutionPending
        blackboardKey {selectedKeyName=Vk_IdleExecutionStatus}
       }
      }]
     properties {waitTime {defaultValue=10000000000}}
    }
    Idle {class=BTTask_Wait, properties {nodeName=Idle, waitTime {defaultValue=1}}}
   ]
   decorators [Loop {class=BTDecorator_Loop}]
   properties {
    nodeName=Loop to prevent evaluator to restart
    children [
     {
      childComposite=&self:BTComposite_Selector_4
      childTask=null
      decorators [&self:BTDecorator_Blackboard_29]
     }
     {
      childComposite=null
      childTask=&self:BTTask_Wait_2
      decorators [&self:BTDecorator_Blackboard_3]
     }
     {childComposite=null, childTask=&self:BTTask_Wait_0, decorators []}
    ]
    services []
   }
   services []
  }]
 decorators []
 properties {
  children [{
    childComposite=&self:BTComposite_Selector_3
    childTask=null
    decorators [&self:BTDecorator_Loop_4]
   }]
  services [&self:FortBTContext_SuppressGoalUpdate_0]
 }
 services [Suppress Goal Updates {
   class=FortBTContext_SuppressGoalUpdate
   properties {
    bCallTickOnSearchStart=false
    bRestartTimerOnEachActivation=false
    bUnregisterFromGoalManager=true
    interval=0.5
    nodeName=Suppress Goal Updates
    randomDeviation=0.1
    enabled {defaultValue=true, key=None}
   }
  }]
}
